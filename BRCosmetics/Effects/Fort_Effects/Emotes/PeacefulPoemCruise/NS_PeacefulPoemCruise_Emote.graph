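# Niagara system for the PeacefulPoemCruise emote that kicks up smoke and debris on land, or water effects in water.
# User.bIsInWater switches between Chunks and SmokeWorld and the drops, foam and splash emitters, mostly spawned by distance traveled.
# Chunks, smoke and drops partition particles across skeletal mesh sockets, while foam and splash meshes snap to the water plane.
# Land emitters add constant spawn rates in the front end, and smoke alpha follows time of day.
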NiagaraSystem /BRCosmetics/Effects/Fort_Effects/Emotes/PeacefulPoemCruise/NS_PeacefulPoemCruise_Emote
dependencies {
 usedDataInterfaces [
  &NiagaraDataInterfaceCurve
  &NiagaraDataInterfaceColorCurve
  &NiagaraDataInterfaceVector2DCurve
  &NiagaraDataInterfaceArrayInt32
  &NiagaraDataInterfaceSkeletalMesh
 ]
 usedDynamicInputs [
  &UniformRangedLinearColor
  &SimulationPosition
  /Game/Effects/Niagara/DynamicInputs/Random/FN_RandomVectorInCone
 ]
 usedModules [
  &self:SystemSpawnScript.SystemScriptSource.SystemScriptGraph.NiagaraNodeAssignment_2.SetVariables_#1
  &SystemState
 ]
}
emitters {
 Chunks {
  emitterUpdateScript [
   &EmitterState(Enable Distance Culling=true, MaxDistance=1125, Module.bUseMaxDistance=true, Module.bUseMinDistance=false, Scalability Mode=::NewEnumerator1)
   &SpawnRate(
    SpawnRate=&Multiply_Float(
     A=&MakeCustomFloatFromBool(Bool=$User.bIsInWater, True Float=0, False Float=1)
     B=&FN_FrontEndFloatModifier(Gameplay=0, Front End=5)
    )
    Use Spawn Probability=false
   )
   &SpawnPerUnit(
    Module.Use Max Movement Threshold=true
    Module.Use Movement Tolerance=true
    Movement Tolerance=1
    Spawn Probability:NiagaraFloat=&MakeCustomFloatFromBool(Bool=$User.bIsInWater, True Float=0, False Float=1)
    Spawn Spacing=6
    Use Spawn Probability=true
   )
  ]
  particleSpawnScript [
   &PartitionParticles(Write Partition to Ribbon ID=::NewEnumerator0)
   &InitializeParticle(
    Color=&ColorFromCurve(
     ColorCurve=DataInterface() {propertyValues {
       AlphaCurve {keys [<at -0.0023 1 RCIM_Cubic>]}
       BlueCurve {keys [<at -0.0023 0.2122 RCIM_Cubic>,<at 0.1383 0.2122>,<at 0.2178 0.1859>,<at 0.4486 0.75>,<at 0.7126 0.1859>,<at 1.0234 0.8802>]}
       GreenCurve {keys [<at -0.0023 0.2325 RCIM_Cubic>,<at 0.1383 0.2325>,<at 0.2178 0.2005>,<at 0.4486 0.8456>,<at 0.7126 0.2122>,<at 1.0234 0.9474>]}
       RedCurve {keys [<at -0.0023 0.25 RCIM_Cubic>,<at 0.1383 0.25>,<at 0.2178 0.2188>,<at 0.4486 1>,<at 0.7126 0.2188>,<at 1.0234 1>]}
      }}
     CurveIndex=&UniformRangedFloat(Minimum=0, Maximum=1)
    )
    Lifetime Max=0.4
    Lifetime Min=0.25
    Lifetime Mode=::NewEnumerator1
    Sprite Rotation Mode=::NewEnumerator2
    Sprite Size Mode=::NewEnumerator1
    Uniform Sprite Size Max=6
    Uniform Sprite Size Min=4
    UsePositionOffset=false
    Write Material Random=true
   )
   &SkeletalMeshLocation(
    Bone / Socket Index=$Particles.PartitionParticles.Partition
    MaskSpawnGroup=false
    Mesh Sampling Type=::NewEnumerator5
    Orientation Sampling=::NewEnumerator1
    Sampled Position Offset=<0 -10 0>
    Skeletal Mesh=$System.SkeletalMesh
    Socket Sampling Mode=::NewEnumerator5
   )
   &ShapeLocation(MaskSpawnGroup=false, Override Seed=false, Sphere Radius=1, UseOwnerScale=true)
   &AddVelocity(Velocity=<0 -75 125>)
   &InheritVelocity(Speed Limit=true, UseSpeedThreshold=false)
   &AddVelocityInCone(
    Cone Angle=50
    Cone Axis=<0 -1 1>
    Module.Use Velocity Falloff On Cone Axis=false
    Velocity Strength=&UniformRangedFloat(Minimum=0, Maximum=250)
   )
  ]
  particleUpdateScript [
   &ParticleState()
   &GravityForce()
   &Drag(Use Linear Drag=true, Use Rotational Drag=true)
   &SolveForcesAndVelocity(Clamp Velocity=false, Limit Acceleration=false)
   &ScaleColor(Scale RGB=<1.2 1.2 1.2>, ScaleA=true, ScaleRGB=true)
   &ScaleSpriteSize(Scale Factor=&Vector2DFromFloat(Value=&FloatFromCurve(FloatCurve=DataInterface() {propertyValues {Curve {keys [
          <at 0.7 0.9701>
          {arriveTangent=-3.2149, interpMode=RCIM_Cubic, leaveTangent=-3.2149, tangentMode=RCTM_User, time=0.9205, value=0.9191}
          <at 1.0011 0.002>
         ]}}})))
   &SpriteRotationRate()
   &ScaleSpriteSizeBySpeed(
    Max Scale Factor=&Vector2DFromFloat(Value=1)
    Min Scale Factor=&Vector2DFromFloat(Value=0.25)
    Velocity Threshold=200
   )
  ]
  rendererClasses [&NiagaraSpriteRendererProperties]
  renderers [{rendererClass=&NiagaraSpriteRendererProperties}]
 }
 FN_Water_Drops_ {
  emitterUpdateScript [
   &EmitterState()
   &SpawnPerUnit(
    Module.Use Max Movement Threshold=true
    Module.Use Movement Tolerance=true
    Movement Tolerance=3
    Spawn Probability:NiagaraFloat=&MakeCustomFloatFromBool(Bool=$User.bIsInWater, True Float=1, False Float=0)
    Spawn Spacing=10
    Use Spawn Probability=true
   )
  ]
  particleSpawnScript [
   &PartitionParticles(Number Of Partitions=4, Write Partition to Ribbon ID=::NewEnumerator0)
   &InitializeParticle(
    Color=<rgba 5 5 5 1>
    Lifetime Max=0.75
    Lifetime Min=0.5
    Lifetime Mode=::NewEnumerator1
    Mass Mode=::NewEnumerator3
    Sprite Size=&Vector2DFromCurve(
     Vector2Curve=DataInterface() {propertyValues {
       XCurve {keys [<at 0 1.0924 RCIM_Cubic>,<at 0.9988 0.3489 RCIM_Cubic>]}
       YCurve {keys [
         {arriveTangent=-6.848, interpMode=RCIM_Cubic, leaveTangent=-6.848, tangentMode=RCTM_User, time=0.0012, value=4.1914}
         {arriveTangent=-2.264, interpMode=RCIM_Cubic, leaveTangent=-2.264, time=0.2419, value=3.2838}
         <at 0.9988 1.9328 RCIM_Cubic>
        ]}
      }}
     CurveIndex=&RandomRangeFloat(Minimum=0, Maximum=1)
     ScaleCurve=&RandomRangeVector2D(Minimum=<1 1>, Maximum=<2 2>)
    )
    Sprite Size Mode=::NewEnumerator4
    UsePositionOffset=false
    Write Material Random=true
   )
   &SkeletalMeshLocation(
    Bone / Socket Index=$Particles.PartitionParticles.Partition
    MaskSpawnGroup=false
    Mesh Sampling Type=::NewEnumerator5
    Sampled Position Offset Coordinate Space:ENiagaraSkelMeshTransforms=::NewEnumerator4
    Skeletal Mesh=$System.SkeletalMesh
    Socket Sampling Mode=::NewEnumerator5
   )
   &ShapeLocation(MaskSpawnGroup=false, Override Seed=false, Sphere Radius=5, UseOwnerScale=true)
   &DynamicMaterialParameters(
    Index 0 Param 1=&UniformRangedFloat(Minimum=4, Maximum=10)
    Module.Index0Param0Write=true
    Module.Index0Param1Write=true
    Module.Index0Param2Write=true
    Module.Index0Param3Write=true
   )
   /SetVariables(Particles.SpriteSize=&Multiply_Vector2(
     A=$Particles.SpriteSize
     B=&Vector2DFromFloat(Value=&FloatFromCurve(
       FloatCurve=DataInterface() {propertyValues {Curve {keys [
           {arriveTangent=1.6964, interpMode=RCIM_Cubic, time=-0.0012, value=-0.0071}
           {arriveTangent=13.7683, interpMode=RCIM_Cubic, leaveTangent=13.7683, time=0.3287, value=17.2263}
           {arriveTangent=-3.3438, interpMode=RCIM_Cubic, leaveTangent=-3.3438, time=0.7139, value=9.8392}
           {arriveTangent=1.8535, interpMode=RCIM_Cubic, leaveTangent=1.8535, tangentMode=RCTM_User, time=0.9945, value=15}
          ]}}}
       CurveIndex=$Emitter.Scalability.DistanceFraction
      ))
    )) {enabled=false, isSetParameters=true}
   &AddVelocityInCone(
    Cone Angle=20
    Cone Axis=&MakeVector(
     X=&Lerp_Float(
      A=0.5
      B=-0.5
      Alpha=&MakeFloatFromInt(Int=&SelectIntFromArray(Direct Array Index=$Particles.PartitionParticles.Partition, Int Selection Array=DataInterface() {propertyValues {IntData [0,1,0,1]}}))
     )
     Y=-1
     Z=2
    )
    Module.Use Velocity Falloff On Cone Axis=true
    Velocity Falloff Away From Cone Axis=0.333
    Velocity Strength=600
   )
   &InheritVelocity(Inherited Velocity Speed Limit=5000, Speed Limit=true, UseSpeedThreshold=false)
  ]
  particleUpdateScript [
   &ParticleState()
   /Game/Effects/Niagara/Emitters/Water/FN_DistanceKillParticles() {enabled=false}
   /Niagara/Modules/Update/Forces/CurlNoiseForce(
    Noise Frequency=1
    Noise Strength=&FloatFromCurve(FloatCurve=DataInterface() {propertyValues {Curve {keys [
         {arriveTangent=-1159.6794, interpMode=RCIM_Cubic, leaveTangent=-1159.6792, tangentMode=RCTM_User, time=-0.0046, value=500}
         {arriveTangent=-447.4286, interpMode=RCIM_Cubic, leaveTangent=-447.4286, time=0.2223, value=189.6793}
         <at 1.0011 50 RCIM_Cubic>
        ]}}})
    Pan Noise=false
   )
   &Drag(Use Linear Drag=true, Use Rotational Drag=true)
   &GravityForce()
   &InheritSourceMovement(
    Applied Position Scale=&VectorFromFloat(Value=&FloatFromCurve(FloatCurve=DataInterface() {propertyValues {Curve {keys [<at 0 1>,<at 1 0>]}}}))
    ApplyForce=false
    ApplyPosition=true
    ApplyPositionFalloff=false
   )
   &SolveForcesAndVelocity(Clamp Velocity=false, Limit Acceleration=false)
   &ScaleColor(
    Scale Alpha=&FloatFromCurve(FloatCurve=DataInterface() {propertyValues {Curve {keys [
         <at 0 0 RCIM_Cubic>
         {arriveTangent=0.002, interpMode=RCIM_Cubic, leaveTangent=0.002, time=0.1183, value=1.0113}
         <at 1.0011 0.002 RCIM_Cubic>
        ]}}})
    ScaleA=true
    ScaleRGB=true
   )
  ]
  rendererClasses [&NiagaraSpriteRendererProperties]
  renderers [{rendererClass=&NiagaraSpriteRendererProperties}]
 }
 FN_Water_Foam_SnapToSurface {
  emitterUpdateScript [
   &EmitterState()
   &SpawnPerUnit(
    Module.Use Max Movement Threshold=true
    Module.Use Movement Tolerance=true
    Movement Tolerance=3
    Spawn Probability:NiagaraFloat=&MakeFloatFromInt(Int=&MakeCustomIntFromBool(Bool=$User.bIsInWater, True INT=1, False INT=0))
    Spawn Spacing=10
    Use Spawn Probability=true
   )
   &SpawnBurst_Instantaneous(
    Spawn Count=&Multiply_Int(A=&MakeCustomIntFromBool(Bool=$User.bIsInWater, True INT=1, False INT=0), B=1)
    Use Spawn Probability=false
    UseLoopCountLimit=false
   )
  ]
  particleSpawnScript [
   &InitializeParticle(
    Lifetime Max=3
    Lifetime Mode=::NewEnumerator1
    Mesh Scale=&VectorFromFloat(Value=&Multiply_Float(A=55, B=$User.EffectScale))
    Mesh Scale Mode=::NewEnumerator4
    Position Offset Coordinate Space=::World
    UsePositionOffset=true
    Write Material Random=true
   )
   &InitialMeshOrientation(
    Mesh Orientation Mode=::NewEnumerator7
    Rotation=&UniformRangedVector(Minimum=<0 0 0>, Maximum=<0 0 1>)
    Use Rotation Vector=true
   )
   &DynamicMaterialParameters(
    Index 1 Param 3=200
    Module.Index1Param0Write=false
    Module.Index1Param1Write=false
    Module.Index1Param2Write=true
    Module.Index1Param3Write=false
    Write Parameter Index 0=false
    Write Parameter Index 1=true
   ) {enabled=false}
   &InitializeWaterSnapToOwner()
   &CameraOffset(Camera Offset Amount=25)
  ]
  particleUpdateScript [
   &ParticleState()
   &DynamicMaterialParameters(
    Index 0 Param 1=&RandomRangeFloat(Minimum=-1, Maximum=1)
    Index 0 Param 2=&FloatFromCurve(FloatCurve=DataInterface() {propertyValues {Curve {keys [<at 0 0.33>,<at 1 1.25>]}}})
    Index 0 Param 3=&FloatFromCurve(
     FloatCurve=DataInterface() {propertyValues {Curve {keys [{tangentMode=RCTM_User, time=0, value=1}, {tangentMode=RCTM_User, time=1, value=0}]}}}
     Scale Curve=1.2
    )
    Index 2 Param 1=&FloatFromCurve(FloatCurve=DataInterface() {propertyValues {Curve {keys [<at -0.0046 0.4>,<at 1.0011 1.5>]}}})
    Index 2 Param 2=&FloatFromCurve(FloatCurve=DataInterface() {propertyValues {Curve {keys [<at -0.0046 0.1>,<at 0.11 2.6>,<at 1.0011 0.3>]}}})
    Index 2 Param 3=1
    Index 2 Param 4=&FloatFromCurve(FloatCurve=DataInterface() {propertyValues {Curve {keys [
         {arriveTangent=-1.9991, interpMode=RCIM_Cubic, time=0, value=1}
         {arriveTangent=-2.7143, interpMode=RCIM_Cubic, leaveTangent=-2.7143, tangentMode=RCTM_User, time=0.2095, value=0.6901}
         <at 1 0 RCIM_Cubic>
        ]}}})
    Module.Index0Param0Write=true
    Module.Index0Param1Write=true
    Module.Index0Param2Write=true
    Module.Index0Param3Write=true
    Module.Index2Param0Write=true
    Module.Index2Param1Write=true
    Module.Index2Param2Write=true
    Module.Index2Param3Write=true
    Write Parameter Index 2=true
   ) {module=DynamicMaterialParameters001}
   &ScaleMeshSize(Scale Factor=&VectorFromFloat(Value=&FloatFromCurve(FloatCurve=DataInterface() {propertyValues {Curve {keys [
          {arriveTangent=26.4511, interpMode=RCIM_Cubic, leaveTangent=26.451, tangentMode=RCTM_User, time=-0.0046, value=0.6073}
          {arriveTangent=3.3733, interpMode=RCIM_Cubic, leaveTangent=3.3733, time=0.1268, value=2.5016}
          <at 1.0011 4 RCIM_Cubic>
         ]}}}))) {module=MeshSizeScale}
   &ConstrainPositionToPlane(
    Plane Normal:Vector3f=$Particles.WaterPlaneNormal
    Plane Position:NiagaraPosition=&AddVectorToPosition(Position=$Particles.Position, Vector=&Add_Vector(A=$Particles.WaterPlaneLocation, B=<0 0 0>))
   )
  ]
  rendererClasses [&NiagaraMeshRendererProperties]
  renderers [{rendererClass=&NiagaraMeshRendererProperties, rendererIndex=0}]
 }
 FN_Water_splash_SnapToSurface {
  emitterUpdateScript [
   &EmitterState()
   &SpawnPerUnit(
    Module.Use Max Movement Threshold=true
    Module.Use Movement Tolerance=true
    Movement Tolerance=3
    Spawn Probability:NiagaraFloat=&MakeFloatFromInt(Int=&MakeCustomIntFromBool(Bool=$User.bIsInWater, True INT=1, False INT=0))
    Spawn Spacing=100
    Use Spawn Probability=true
   )
  ]
  particleSpawnScript [
   &InitializeParticle(
    Lifetime=&UniformRangedFloat(Minimum=0.7, Maximum=0.9)
    Mass=&UniformRangedFloat(Minimum=0.6, Maximum=1.2)
    Mass Mode=::NewEnumerator1
    Mesh Scale=<0.3 0.3 0.1>
    Mesh Scale Mode=::NewEnumerator4
    UsePositionOffset=false
    Write Material Random=true
   )
   &InitialMeshOrientation(
    Mesh Orientation Mode=::NewEnumerator1
    Rotation=&UniformRangedVector(Minimum=<0 0 0>, Maximum=<0 0 360>)
    Use Rotation Vector=true
   )
   &CylinderLocation(
    Cylinder Height=1
    Cylinder Radius=10
    Lathe Profile=DataInterface() {propertyValues {Curve {keys []}}}
    MaskSpawnGroup=false
    Module.Enable Lathe Profile=false
    Module.Override Randomness=false
    Module.Surface Only=false
   )
   /SetVariables(Particles.Position=&Add_Vector(
     A=Linked() {linkedVariable {name=Particles.Position, type=NiagaraPosition}}
     B=&Multiply_Vector(A=$Engine.Owner.Velocity, B=<0.1 0.1 0>)
    )) {isSetParameters=true}
   &InheritVelocity(Module.Limit Speed=true, Velocity Limit=200, Velocity Scale=<0.5 0.5 0>)
   &InitializeWaterSnapToOwner()
  ]
  particleUpdateScript [
   &ParticleState()
   &DynamicMaterialParameters(
    Index 0 Param 1=&FloatFromCurve(FloatCurve=DataInterface() {propertyValues {Curve {keys [
         <at -0.0046 -2>
         {arriveTangent=1.9886, interpMode=RCIM_Cubic, leaveTangent=1.9886, time=0.7384, value=-0.7816}
         {arriveTangent=6.4587, interpMode=RCIM_Cubic, leaveTangent=6.4587, tangentMode=RCTM_User, time=1.0011, value=0}
        ]}}})
    Index 0 Param 2=&FloatFromCurve(FloatCurve=DataInterface() {propertyValues {Curve {keys [<at -0.0046 4>,<at 1.0011 0>]}}})
    Index 0 Param 3=&FloatFromCurve(FloatCurve=DataInterface() {propertyValues {Curve {keys [
         {arriveTangent=0.096, interpMode=RCIM_Cubic, leaveTangent=0.096, tangentMode=RCTM_User, time=-0.0046, value=0.0882}
         <at 1.0011 0.1372 RCIM_Cubic>
        ]}}})
    Index 0 Param 4=&FloatFromCurve(FloatCurve=DataInterface() {propertyValues {Curve {keys [
         {arriveTangent=1.9125, interpMode=RCIM_Cubic, leaveTangent=0.2728, tangentMode=RCTM_User, time=-0.0046, value=0.9276}
         {arriveTangent=4.6326, interpMode=RCIM_Cubic, leaveTangent=4.6326, tangentMode=RCTM_User, time=1.0011, value=2.7648}
        ]}}})
    Index 1 Param 1=&FloatFromCurve(FloatCurve=DataInterface() {propertyValues {Curve {keys [
         {interpMode=RCIM_Cubic, tangentMode=RCTM_User, time=0, value=0}
         {arriveTangent=3.1921, interpMode=RCIM_Cubic, leaveTangent=3.1921, tangentMode=RCTM_User, time=1, value=1}
        ]}}})
    Module.Index0Param0Write=true
    Module.Index0Param1Write=true
    Module.Index0Param2Write=true
    Module.Index0Param3Write=true
    Module.Index1Param0Write=true
    Module.Index1Param1Write=true
    Module.Index1Param2Write=true
    Module.Index1Param3Write=true
    Write Parameter Index 1=true
   )
   &ScaleColor(
    Scale Alpha=&FloatFromCurve(FloatCurve=DataInterface() {propertyValues {Curve {keys [<at -0.0046 1.0113 RCIM_Cubic>,<at 1.0011 0.002 RCIM_Cubic>]}}})
    ScaleA=true
    ScaleRGB=false
   )
   &Drag(Drag=2, Use Linear Drag=true, Use Rotational Drag=false)
   &SolveForcesAndVelocity(Clamp Velocity=false, Limit Acceleration=false)
   &ScaleMeshSize(Scale Factor=&VectorFromFloat(Value=&FloatFromCurve(FloatCurve=DataInterface() {propertyValues {Curve {keys [<at 0 1>,<at 1 1.2>]}}}))) {module=MeshSizeScale}
   &ConstrainPositionToPlane(
    Plane Normal:Vector3f=$Particles.WaterPlaneNormal
    Plane Position:NiagaraPosition=&AddVectorToPosition(Position=$Particles.Position, Vector=&Add_Vector(A=$Particles.WaterPlaneLocation, B=<0 0 0>))
   )
  ]
  rendererClasses [&NiagaraMeshRendererProperties]
  renderers [{rendererClass=&NiagaraMeshRendererProperties, rendererIndex=0}]
 }
 SmokeWorld {
  emitterUpdateScript [
   &EmitterState()
   &SpawnRate(
    SpawnRate=&Multiply_Float(
     A=&MakeCustomFloatFromBool(Bool=$User.bIsInWater, True Float=0, False Float=1)
     B=&FN_FrontEndFloatModifier(Gameplay=0, Front End=40)
    )
    Use Spawn Probability=false
   )
   &SpawnPerUnit(
    Module.Use Max Movement Threshold=true
    Module.Use Movement Tolerance=true
    Movement Tolerance=3
    Spawn Probability:NiagaraFloat=&MakeCustomFloatFromBool(Bool=$User.bIsInWater, True Float=0, False Float=1)
    Spawn Spacing=5
    Use Spawn Probability=true
   )
  ]
  particleSpawnScript [
   &PartitionParticles(Number Of Partitions=4, Write Partition to Ribbon ID=::NewEnumerator0)
   &InitializeParticle(
    Color Maximum=<rgba 0.2 0.1987 0.192 0.225>
    Color Minimum=<rgba 0.25 0.2432 0.2322 0.2>
    Color Mode=::NewEnumerator2
    Lifetime=&Lerp_Float(
     A=&RandomRangeFloat(Minimum=0.6, Maximum=0.75)
     B=&RandomRangeFloat(Minimum=0.45, Maximum=0.55)
     Alpha=&MakeCustomFloatFromBool(
      Bool=&SetBoolByIntComparison(A=$Particles.PartitionParticles.Partition, B=2, Comparison Type=::NewEnumerator2)
      True Float=1
      False Float=0
     )
    )
    Sprite Rotation Mode=::NewEnumerator2
    Sprite Size Mode=::NewEnumerator1
    Uniform Sprite Size Max=25
    Uniform Sprite Size Min=15
    UsePositionOffset=false
    Write Material Random=true
   )
   &SkeletalMeshLocation(
    Bone / Socket Index=$Particles.PartitionParticles.Partition
    MaskSpawnGroup=false
    Mesh Sampling Type=::NewEnumerator5
    Orientation Sampling=::NewEnumerator1
    Skeletal Mesh=$System.SkeletalMesh
    Socket Sampling Mode=::NewEnumerator5
   )
   &ShapeLocation(MaskSpawnGroup=false, Override Seed=false, Sphere Radius=2, UseOwnerScale=true)
   &InheritVelocity(Module.Limit Speed=true, Velocity Limit=1000)
   &AddVelocity(Velocity=&RandomRangeVector(Minimum=<0 -300 50>, Maximum=<0 -300 75>))
   &ApplyInitialForces()
   &CameraOffset(Camera Offset Amount=15)
  ]
  particleUpdateScript [
   &ParticleState()
   &AccelerationForce(Acceleration=&Multiply_VectorByFloat(
     Vector=<0 0 500>
     Float=&FloatFromCurve(FloatCurve=DataInterface() {propertyValues {Curve {keys [
          {arriveTangent=1.8323, interpMode=RCIM_Cubic, leaveTangent=1.8323, tangentMode=RCTM_User, time=0, value=0}
          {interpMode=RCIM_Cubic, tangentMode=RCTM_User, time=1, value=1}
         ]}}})
    ))
   &Drag(Drag=10, Use Linear Drag=true, Use Rotational Drag=true)
   &SolveForcesAndVelocity(Clamp Velocity=false, Limit Acceleration=false)
   &ScaleSpriteSize(Scale Factor=&Vector2DFromFloat(Value=&FloatFromCurve(FloatCurve=DataInterface() {propertyValues {Curve {keys [
          {arriveTangent=1.491, interpMode=RCIM_Cubic, leaveTangent=1.491, tangentMode=RCTM_User, time=-0.0046, value=1.1969}
          <at 0.75 3.4809 RCIM_Cubic>
         ]}}})))
   &DynamicMaterialParameters(
    Index 0 Param 1=&Add_Float(
     A=&RandomRangeFloat(Minimum=0, Maximum=1)
     B=&FloatFromCurve(
      FloatCurve=DataInterface() {propertyValues {Curve {keys [
          <at 0 0 RCIM_Cubic>
          {arriveTangent=1.5947, interpMode=RCIM_Cubic, leaveTangent=1.5947, tangentMode=RCTM_User, time=0.2557, value=0.5257}
          <at 1 1 RCIM_Cubic>
         ]}}}
      CurveIndex=$Particles.Age
      Scale Curve=0.25
     )
    )
    Module.Index0Param0Write=true
    Module.Index0Param1Write=true
    Module.Index0Param2Write=true
    Module.Index0Param3Write=true
   ) {module=DynamicMatParams}
   &ScaleColor(
    Scale Alpha=&Multiply_Float(
     A=&FloatFromCurve(
      FloatCurve=DataInterface() {propertyValues {Curve {keys [
          {interpMode=RCIM_Cubic, tangentMode=RCTM_User, time=0, value=0}
          <at 0.15 1 RCIM_Cubic>
          {arriveTangent=-4.4473, tangentMode=RCTM_User, time=0.9, value=0}
         ]}}}
      Scale Curve=&FN_FrontEndFloatModifier(Gameplay=0.75, Front End=1)
     )
     B=&FloatFromCurve(
      FloatCurve=DataInterface() {propertyValues {Curve {keys [
          {arriveTangent=0.4052, interpMode=RCIM_Cubic, leaveTangent=0.4052, tangentMode=RCTM_User, time=0, value=0.9}
          {interpMode=RCIM_Cubic, tangentMode=RCTM_User, time=0.5, value=1}
          {arriveTangent=-0.4431, tangentMode=RCTM_User, time=1, value=0.9}
         ]}}}
      CurveIndex=Linked() {linkedVariable {name=NPC.FortniteNPC.TimeOfDay_0-1, type=NiagaraFloat}}
     )
    )
    ScaleA=true
    ScaleRGB=false
   )
  ]
  rendererClasses [&NiagaraSpriteRendererProperties]
  renderers [
   {rendererClass=&NiagaraSpriteRendererProperties}
   {rendererClass=&NiagaraSpriteRendererProperties, rendererIndex=1}
  ]
 }
}
summary {
 systemName=NS_PeacefulPoemCruise_Emote
 emitters [
  {emitterName=SmokeWorld, rendererClasses [&NiagaraSpriteRendererProperties]}
  {emitterName=FN_Water_Foam_SnapToSurface, rendererClasses [&NiagaraMeshRendererProperties]}
  {emitterName=Chunks, rendererClasses [&NiagaraSpriteRendererProperties]}
  {emitterName=FN_Water_Drops_, rendererClasses [&NiagaraSpriteRendererProperties]}
  {emitterName=FN_Water_splash_SnapToSurface, rendererClasses [&NiagaraMeshRendererProperties]}
 ]
 userVariables [
  User.bIsFrontEnd {type=NiagaraBool}
  User.bIsFrontEndPreview {type=NiagaraBool}
  User.bIsInWater {type=NiagaraBool}
  User.EffectScale {type=NiagaraFloat, defaultValue=1, description=""}
  User.WaterBodyIndex {type=NiagaraFloat, defaultValue=0, description=""}
  User.WaterDepth {type=NiagaraFloat, defaultValue=0, description=""}
  User.WaterPlaneLocation {type=NiagaraPosition, description="", defaultValue {}}
  User.WaterPlaneLocation {type=Vector3f, description="", defaultValue {struct=Vector3f, value=<0 0 0>}}
  User.WaterPlaneNormal {type=Vector3f, description="", defaultValue {struct=Vector3f, value=<0 0 0>}}
  User.WaveAttenuationFactor {type=NiagaraFloat, defaultValue=0, description=""}
  User.WaveTime {type=NiagaraFloat, defaultValue=0, description=""}
 ]
}
