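# Niagara skid effect for the PeacefulPoemCruise emote, spawning from skeletal mesh sockets with particle partitions mirroring the spray between two sides.
# On land it emits debris chunks, smoke and velocity-gated ribbon skid marks that fade over the system loop.
# User.bIsInWater swaps these for water drops and mesh splashes, while tire streak sprites run in both cases.
# Time of day adjusts smoke and streaks; front-end modifiers tune chunks and smoke.
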NiagaraSystem /BRCosmetics/Effects/Fort_Effects/Emotes/PeacefulPoemCruise/NS_PeacefulPoemCruise_Skid_Emote
dependencies {
 usedDataInterfaces [
  &NiagaraDataInterfaceCurve
  &NiagaraDataInterfaceVectorCurve
  &NiagaraDataInterfaceVector2DCurve
  &NiagaraDataInterfaceColorCurve
  &NiagaraDataInterfaceSkeletalMesh
 ]
 usedDynamicInputs [&SimulationPosition, &NormalizedAngleToDegrees]
 usedModules [
  &self:SystemSpawnScript.SystemScriptSource.SystemScriptGraph.NiagaraNodeAssignment_2.SetVariables_#1
  &SystemState
 ]
}
emitters {
 Chunks {
  emitterUpdateScript [
   &EmitterState(Enable Distance Culling=true, MaxDistance=1125, Module.bUseMaxDistance=true, Module.bUseMinDistance=false, Scalability Mode=::NewEnumerator1)
   &SpawnRate(
    SpawnRate=&Multiply_Float(
     A=&MakeCustomFloatFromBool(Bool=$User.bIsInWater, True Float=0, False Float=1)
     B=&FN_FrontEndFloatModifier(Gameplay=20, Front End=5)
    )
    Use Spawn Probability=false
   )
  ]
  particleSpawnScript [
   &PartitionParticles(Write Partition to Ribbon ID=::NewEnumerator0)
   &InitializeParticle(
    Color=&ColorFromCurve(
     ColorCurve=DataInterface() {propertyValues {
       AlphaCurve {keys [<at -0.0023 1 RCIM_Cubic>]}
       BlueCurve {keys [<at -0.0023 0.2122 RCIM_Cubic>,<at 0.1383 0.2122>,<at 0.2178 0.1859>,<at 0.4486 0.75>,<at 0.7126 0.1859>,<at 1.0234 0.8802>]}
       GreenCurve {keys [<at -0.0023 0.2325 RCIM_Cubic>,<at 0.1383 0.2325>,<at 0.2178 0.2005>,<at 0.4486 0.8456>,<at 0.7126 0.2122>,<at 1.0234 0.9474>]}
       RedCurve {keys [<at -0.0023 0.25 RCIM_Cubic>,<at 0.1383 0.25>,<at 0.2178 0.2188>,<at 0.4486 1>,<at 0.7126 0.2188>,<at 1.0234 1>]}
      }}
     CurveIndex=&UniformRangedFloat(Minimum=0, Maximum=1)
    )
    Lifetime Max=0.4
    Lifetime Min=0.25
    Lifetime Mode=::NewEnumerator1
    Sprite Rotation Mode=::NewEnumerator2
    Sprite Size Mode=::NewEnumerator1
    Uniform Sprite Size Min=4
    UsePositionOffset=false
    Write Material Random=true
   )
   &SkeletalMeshLocation(
    Bone / Socket Index=$Particles.PartitionParticles.Partition
    MaskSpawnGroup=false
    Mesh Sampling Type=::NewEnumerator5
    Orientation Sampling=::NewEnumerator1
    Sampled Position Offset=<0 -10 0>
    Skeletal Mesh=$System.SkeletalMesh
    Socket Sampling Mode=::NewEnumerator5
   )
   &ShapeLocation(MaskSpawnGroup=false, Override Seed=false, Sphere Radius=1, UseOwnerScale=true)
   &AddVelocityInCone(
    Cone Angle=20
    Cone Axis=&MakeVector(
     X=&Lerp_Float(A=0.5, B=-0.5, Alpha=&MakeFloatFromInt(Int=$Particles.PartitionParticles.Partition))
     Y=-1
     Z=1
    )
    Module.Use Velocity Falloff On Cone Axis=true
    Velocity Falloff Away From Cone Axis=0.333
    Velocity Strength=600
   )
   &InheritVelocity(Speed Limit=true, UseSpeedThreshold=false)
  ]
  particleUpdateScript [
   &ParticleState()
   &GravityForce()
   &Drag(Use Linear Drag=true, Use Rotational Drag=true)
   &SolveForcesAndVelocity(Clamp Velocity=false, Limit Acceleration=false)
   &ScaleColor(Scale RGB=<1.2 1.2 1.2>, ScaleA=true, ScaleRGB=true)
   &ScaleSpriteSize(Scale Factor=&Vector2DFromFloat(Value=&FloatFromCurve(FloatCurve=DataInterface() {propertyValues {Curve {keys [
          <at 0.7 0.9701>
          {arriveTangent=-3.2149, interpMode=RCIM_Cubic, leaveTangent=-3.2149, tangentMode=RCTM_User, time=0.9205, value=0.9191}
          <at 1.0011 0.002>
         ]}}})))
   &SpriteRotationRate()
   &ScaleSpriteSizeBySpeed(
    Max Scale Factor=&Vector2DFromFloat(Value=1)
    Min Scale Factor=&Vector2DFromFloat(Value=0.25)
    Velocity Threshold=200
   )
  ]
  rendererClasses [&NiagaraSpriteRendererProperties]
  renderers [{rendererClass=&NiagaraSpriteRendererProperties}]
 }
 FN_Water_Drops_ {
  emitterUpdateScript [
   &EmitterState()
   &SpawnRate(
    SpawnRate=&Multiply_Float(A=&MakeCustomFloatFromBool(Bool=$User.bIsInWater, True Float=1, False Float=0), B=50)
    Use Spawn Probability=false
   )
  ]
  particleSpawnScript [
   &PartitionParticles(Write Partition to Ribbon ID=::NewEnumerator0)
   &InitializeParticle(
    Color=<rgba 5 5 5 1>
    Lifetime Max=0.75
    Lifetime Min=0.5
    Lifetime Mode=::NewEnumerator1
    Mass Mode=::NewEnumerator3
    Sprite Size=&Vector2DFromCurve(
     Vector2Curve=DataInterface() {propertyValues {
       XCurve {keys [<at 0 1.0924 RCIM_Cubic>,<at 0.9988 0.3489 RCIM_Cubic>]}
       YCurve {keys [
         {arriveTangent=-6.848, interpMode=RCIM_Cubic, leaveTangent=-6.848, tangentMode=RCTM_User, time=0.0012, value=4.1914}
         {arriveTangent=-2.264, interpMode=RCIM_Cubic, leaveTangent=-2.264, time=0.2419, value=3.2838}
         <at 0.9988 1.9328 RCIM_Cubic>
        ]}
      }}
     CurveIndex=&RandomRangeFloat(Minimum=0, Maximum=1)
     ScaleCurve=&RandomRangeVector2D(Minimum=<1 1>, Maximum=<2 2>)
    )
    Sprite Size Mode=::NewEnumerator4
    UsePositionOffset=false
    Write Material Random=true
   )
   &SkeletalMeshLocation(
    Bone / Socket Index=$Particles.PartitionParticles.Partition
    MaskSpawnGroup=false
    Mesh Sampling Type=::NewEnumerator5
    Sampled Position Offset Coordinate Space:ENiagaraSkelMeshTransforms=::NewEnumerator4
    Skeletal Mesh=$System.SkeletalMesh
    Socket Sampling Mode=::NewEnumerator5
   )
   &ShapeLocation(MaskSpawnGroup=false, Override Seed=false, Sphere Radius=5, UseOwnerScale=true)
   &DynamicMaterialParameters(
    Index 0 Param 1=&UniformRangedFloat(Minimum=4, Maximum=10)
    Module.Index0Param0Write=true
    Module.Index0Param1Write=true
    Module.Index0Param2Write=true
    Module.Index0Param3Write=true
   )
   /SetVariables(Particles.SpriteSize=&Multiply_Vector2(
     A=$Particles.SpriteSize
     B=&Vector2DFromFloat(Value=&FloatFromCurve(
       FloatCurve=DataInterface() {propertyValues {Curve {keys [
           {arriveTangent=1.6964, interpMode=RCIM_Cubic, time=-0.0012, value=-0.0071}
           {arriveTangent=13.7683, interpMode=RCIM_Cubic, leaveTangent=13.7683, time=0.3287, value=17.2263}
           {arriveTangent=-3.3438, interpMode=RCIM_Cubic, leaveTangent=-3.3438, time=0.7139, value=9.8392}
           {arriveTangent=1.8535, interpMode=RCIM_Cubic, leaveTangent=1.8535, tangentMode=RCTM_User, time=0.9945, value=15}
          ]}}}
       CurveIndex=$Emitter.Scalability.DistanceFraction
      ))
    )) {enabled=false, isSetParameters=true}
   &AddVelocityInCone(
    Cone Angle=20
    Cone Axis=&MakeVector(
     X=&Lerp_Float(A=0.5, B=-0.5, Alpha=&MakeFloatFromInt(Int=$Particles.PartitionParticles.Partition))
     Y=-1
     Z=2
    )
    Module.Use Velocity Falloff On Cone Axis=true
    Velocity Falloff Away From Cone Axis=0.333
    Velocity Strength=600
   )
   &InheritVelocity(Speed Limit=true, UseSpeedThreshold=false)
  ]
  particleUpdateScript [
   &ParticleState()
   /Game/Effects/Niagara/Emitters/Water/FN_DistanceKillParticles() {enabled=false}
   /Niagara/Modules/Update/Forces/CurlNoiseForce(
    Noise Frequency=1
    Noise Strength=&FloatFromCurve(FloatCurve=DataInterface() {propertyValues {Curve {keys [
         {arriveTangent=-1159.6794, interpMode=RCIM_Cubic, leaveTangent=-1159.6792, tangentMode=RCTM_User, time=-0.0046, value=500}
         {arriveTangent=-447.4286, interpMode=RCIM_Cubic, leaveTangent=-447.4286, time=0.2223, value=189.6793}
         <at 1.0011 50 RCIM_Cubic>
        ]}}})
    Pan Noise=false
   )
   &Drag(Use Linear Drag=true, Use Rotational Drag=true)
   &GravityForce()
   &SolveForcesAndVelocity(Clamp Velocity=false, Limit Acceleration=false)
   &ScaleColor(
    Scale Alpha=&FloatFromCurve(FloatCurve=DataInterface() {propertyValues {Curve {keys [
         <at 0 0 RCIM_Cubic>
         {arriveTangent=0.002, interpMode=RCIM_Cubic, leaveTangent=0.002, time=0.1183, value=1.0113}
         <at 1.0011 0.002 RCIM_Cubic>
        ]}}})
    ScaleA=true
    ScaleRGB=true
   )
  ]
  rendererClasses [&NiagaraSpriteRendererProperties]
  renderers [{rendererClass=&NiagaraSpriteRendererProperties}]
 }
 SkidMarks {
  emitterUpdateScript [
   &EmitterState()
   &SpawnRate(
    SpawnRate=&MakeCustomFloatFromBool(
     Bool=&SetBoolByFloatComparison(
      A=&Abs_Float(Float:NiagaraFloat=&MakeFloatFromVector(Vector=$Engine.Owner.Velocity, Channel=::NewEnumerator0))
      B=0.25
      Comparison Type=::NewEnumerator2
     )
     True Float=500
     False Float=0
    )
    Use Spawn Probability=false
   )
  ]
  particleSpawnScript [
   &PartitionParticles(Write Partition to Ribbon ID=::NewEnumerator0)
   &InitializeParticle(
    Color=&MultiplyLinearColor(
     Linear Color=<rgba 0.01 0.01 0.01 0.65>
     Scale Factor=&MakeFloat4FromVectorAndFloat(
      "Vector (RGB)"=<1 1 1>
      "Float (Alpha)"=&FloatFromCurve(
       FloatCurve=DataInterface() {propertyValues {Curve {keys [
           {interpMode=RCIM_Cubic, tangentMode=RCTM_User, time=0, value=0}
           <at 0.2 1>
           {arriveTangent=-1.25, interpMode=RCIM_Cubic, leaveTangent=-1.25, time=0.7, value=1}
           <at 1 0 RCIM_Cubic>
          ]}}}
       CurveIndex=$System.NormalizedLoopAge
      )
     )
    )
    Lifetime=1.5
    UsePositionOffset=false
    Write Material Random=true
   )
   /SetVariables() {isSetParameters=true}
   /SetVariables(Particles.RibbonWidth=12) {isSetParameters=true}
   &SkeletalMeshLocation(
    Bone / Socket Index=$Particles.PartitionParticles.Partition
    MaskSpawnGroup=false
    Mesh Sampling Type=::NewEnumerator5
    Orientation Sampling=::NewEnumerator1
    Sampled Position Offset=<0 0 1.25>
    Skeletal Mesh=$System.SkeletalMesh
    Socket Sampling Mode=::NewEnumerator5
   )
  ]
  particleUpdateScript [
   &ParticleState()
   &ScaleColor(
    Scale Alpha=&Multiply_Float(
     A=&FloatFromCurve(FloatCurve=DataInterface() {propertyValues {Curve {keys [
          {tangentMode=RCTM_User, time=0, value=1}
          {arriveTangent=-0.0405, interpMode=RCIM_Cubic, leaveTangent=-0.0405, tangentMode=RCTM_User, time=0.7009, value=1}
          {arriveTangent=-8.0089, interpMode=RCIM_Cubic, leaveTangent=-8.0089, tangentMode=RCTM_User, time=1, value=0}
         ]}}})
     B=1
    )
    ScaleA=true
    ScaleRGB=true
   )
   &KillParticles(Kill Particles=$User.bIsInWater)
  ]
  rendererClasses [&NiagaraRibbonRendererProperties]
  renderers [{rendererClass=&NiagaraRibbonRendererProperties}]
 }
 SmokeWorld {
  emitterUpdateScript [
   &EmitterState()
   &SpawnRate(
    SpawnRate=&Multiply_Float(A=&MakeCustomFloatFromBool(Bool=$User.bIsInWater, True Float=0, False Float=1), B=40)
    Use Spawn Probability=false
   )
  ]
  particleSpawnScript [
   &PartitionParticles(Write Partition to Ribbon ID=::NewEnumerator0)
   &InitializeParticle(
    Color Maximum=<rgba 0.2 0.1987 0.192 0.25>
    Color Minimum=<rgba 0.25 0.2432 0.2322 0.2>
    Color Mode=::NewEnumerator2
    Lifetime Max=1.1
    Lifetime Min=0.9
    Lifetime Mode=::NewEnumerator1
    Sprite Rotation Mode=::NewEnumerator2
    Sprite Size Mode=::NewEnumerator1
    Uniform Sprite Size Max=25
    Uniform Sprite Size Min=15
    UsePositionOffset=false
    Write Material Random=true
   )
   &SkeletalMeshLocation(
    Bone / Socket Index=$Particles.PartitionParticles.Partition
    MaskSpawnGroup=false
    Mesh Sampling Type=::NewEnumerator5
    Orientation Sampling=::NewEnumerator1
    Skeletal Mesh=$System.SkeletalMesh
    Socket Sampling Mode=::NewEnumerator5
   )
   &ShapeLocation(MaskSpawnGroup=false, Override Seed=false, Sphere Radius=2, UseOwnerScale=true)
   &AddVelocityInCone(
    Cone Angle=20
    Cone Axis=&MakeVector(
     X=&Lerp_Float(A=0.5, B=-0.5, Alpha=&MakeFloatFromInt(Int=$Particles.PartitionParticles.Partition))
     Y=-1
     Z=0.5
    )
    Module.Use Velocity Falloff On Cone Axis=true
    Velocity Falloff Away From Cone Axis=0.333
    Velocity Strength=300
   )
   &InheritVelocity(Module.Limit Speed=true, Velocity Limit=1000)
   &CameraOffset(Camera Offset Amount=5)
  ]
  particleUpdateScript [
   &ParticleState()
   &Drag(Drag=3, Use Linear Drag=true, Use Rotational Drag=true)
   &SolveForcesAndVelocity(Clamp Velocity=false, Limit Acceleration=false)
   &ScaleSpriteSize(Scale Factor=&Vector2DFromFloat(Value=&FloatFromCurve(FloatCurve=DataInterface() {propertyValues {Curve {keys [{arriveTangent=1.491, interpMode=RCIM_Cubic, leaveTangent=1.491, tangentMode=RCTM_User, time=-0.0046, value=1.1969}, <at 1 3 RCIM_Cubic>]}}})))
   &DynamicMaterialParameters(
    Index 0 Param 1=&Add_Float(
     A=&RandomRangeFloat(Minimum=0, Maximum=1)
     B=&FloatFromCurve(
      FloatCurve=DataInterface() {propertyValues {Curve {keys [
          <at 0 0 RCIM_Cubic>
          {arriveTangent=1.5947, interpMode=RCIM_Cubic, leaveTangent=1.5947, tangentMode=RCTM_User, time=0.2557, value=0.5257}
          <at 1 1 RCIM_Cubic>
         ]}}}
      CurveIndex=$Particles.Age
      Scale Curve=0.25
     )
    )
    Module.Index0Param0Write=true
    Module.Index0Param1Write=true
    Module.Index0Param2Write=true
    Module.Index0Param3Write=true
   ) {module=DynamicMatParams}
   &ScaleColor(
    Scale Alpha=&Multiply_Float(
     A=&FloatFromCurve(
      FloatCurve=DataInterface() {propertyValues {Curve {keys [
          {interpMode=RCIM_Cubic, tangentMode=RCTM_User, time=0, value=0}
          <at 0.15 1 RCIM_Cubic>
          {arriveTangent=-4.4473, tangentMode=RCTM_User, time=0.9, value=0}
         ]}}}
      Scale Curve=&FN_FrontEndFloatModifier(Gameplay=0.75, Front End=1)
     )
     B=&FloatFromCurve(
      FloatCurve=DataInterface() {propertyValues {Curve {keys [
          {arriveTangent=1.1874, interpMode=RCIM_Cubic, leaveTangent=1.1874, tangentMode=RCTM_User, time=0, value=0.8}
          {interpMode=RCIM_Cubic, tangentMode=RCTM_User, time=0.5, value=1}
          {arriveTangent=-1.3304, tangentMode=RCTM_User, time=1, value=0.8}
         ]}}}
      CurveIndex=Linked() {linkedVariable {name=NPC.FortniteNPC.TimeOfDay_0-1, type=NiagaraFloat}}
     )
    )
    ScaleA=true
    ScaleRGB=false
   )
  ]
  rendererClasses [&NiagaraSpriteRendererProperties]
  renderers [
   {rendererClass=&NiagaraSpriteRendererProperties}
   {rendererClass=&NiagaraSpriteRendererProperties, rendererIndex=1}
  ]
 }
 Tire_Skidstreaks {
  emitterUpdateScript [
   &EmitterState()
   &SpawnRate(SpawnRate=30, Use Spawn Probability=false)
  ]
  particleSpawnScript [
   &PartitionParticles()
   &InitializeParticle(
    Color=<rgba 0.15 0.15 0.15 0.5>
    Lifetime=&UniformRangedFloat(Minimum=0.11, Maximum=0.14)
    Mass Mode=::NewEnumerator1
    Sprite Size=&MakeVector2D(
     X=&UniformRangedFloat(Minimum=12, Maximum=16)
     Y=&UniformRangedFloat(Minimum=42, Maximum=56)
    )
    Sprite Size Mode=::NewEnumerator4
    UsePositionOffset=false
    Write Material Random=true
   )
   &AddVelocityInCone(
    Cone Angle=20
    Cone Axis=&MakeVector(
     X=&Lerp_Float(A=0.5, B=-0.5, Alpha=&MakeFloatFromInt(Int=$Particles.PartitionParticles.Partition))
     Y=-1
     Z=1
    )
    Module.Use Velocity Falloff On Cone Axis=true
    Velocity Falloff Away From Cone Axis=0.333
    Velocity Strength=600
   )
   /Niagara/Modules/Update/SubUV/SubUVAnimation(Number Of Frames=3, SubUV Animation Mode=::NewEnumerator1)
   &CameraOffset(Camera Offset Amount=10)
  ]
  particleUpdateScript [
   &ParticleState()
   &SkeletalMeshLocation(
    Bone / Socket Index=$Particles.PartitionParticles.Partition
    MaskSpawnGroup=false
    Mesh Sampling Type=::NewEnumerator5
    Sampled Position Offset=<2 2 0>
    Sampled Position Offset Coordinate Space:ENiagaraSkelMeshTransforms=::NewEnumerator4
    Skeletal Mesh=$System.SkeletalMesh
    Socket Sampling Mode=::NewEnumerator5
   )
   &SolveForcesAndVelocity(Clamp Velocity=false, Limit Acceleration=false)
   &DynamicMaterialParameters(
    Index 0 Param 1=&FloatFromCurve(FloatCurve=DataInterface() {propertyValues {Curve {keys [<at 0 1>,<at 1.0011 -1>]}}})
    Index 0 Param 4=0.5
    Module.Param0WriteEnabled=true
    Module.Param1WriteEnabled=true
    Module.Param2WriteEnabled=true
    Module.Param3WriteEnabled=true
   )
   &ScaleColor(
    Scale Alpha=&FloatFromCurve(
     FloatCurve=DataInterface() {propertyValues {Curve {keys [<at 0 0>, {arriveTangent=8.6954, interpMode=RCIM_Cubic, leaveTangent=8.6954, tangentMode=RCTM_User, time=0.0788, value=1.0113}, <at 1 0>]}}}
     Scale Curve=&FloatFromCurve(
      FloatCurve=DataInterface() {propertyValues {Curve {keys [
          {arriveTangent=1.1874, interpMode=RCIM_Cubic, time=0, value=0.8}
          {interpMode=RCIM_Cubic, tangentMode=RCTM_User, time=0.5, value=1}
          {interpMode=RCIM_Cubic, leaveTangent=-1.1852, time=1, value=0.8}
         ]}}}
      CurveIndex=Linked() {linkedVariable {name=NPC.FortniteNPC.TimeOfDay_0-1, type=NiagaraFloat}}
     )
    )
    Scale RGB=&MakeVectorFromLinearColorRGB(Color=&ColorFromCurve(
      ColorCurve=DataInterface() {propertyValues {
        AlphaCurve {keys [<at 0 1 RCIM_Cubic>]}
        BlueCurve {keys [
          <at 0 1>
          {arriveTangent=-0.8476, interpMode=RCIM_Cubic, leaveTangent=-0.8476, time=0.25, value=0.1465}
          <at 0.5 0.5762>
          <at 0.75 0.147>
          <at 1 1 RCIM_Cubic>
         ]}
        GreenCurve {keys [
          <at 0 0.4342>
          {arriveTangent=0.5735, interpMode=RCIM_Cubic, leaveTangent=0.5735, time=0.25, value=0.2769}
          <at 0.5 0.7209>
          <at 0.75 0.2789>
          <at 1 0.4342 RCIM_Cubic>
         ]}
        RedCurve {keys [
          <at 0 0.2232>
          {arriveTangent=1.1473, interpMode=RCIM_Cubic, leaveTangent=1.1473, time=0.25, value=0.7031}
          <at 0.5 0.7969>
          <at 0.75 0.7011>
          <at 1 0.2232 RCIM_Cubic>
         ]}
       }}
      CurveIndex=Linked() {linkedVariable {name=NPC.FortniteNPC.TimeOfDay_0-1, type=NiagaraFloat}}
     ))
    ScaleA=true
    ScaleRGB=true
   )
   &ScaleSpriteSize(Uniform Curve Sprite Scale=DataInterface() {propertyValues {Curve {keys [{arriveTangent=2.3953, interpMode=RCIM_Cubic, leaveTangent=2.3953, tangentMode=RCTM_User, time=0, value=0.2}, <at 1 1>]}}})
  ]
  rendererClasses [&NiagaraSpriteRendererProperties]
  renderers [{rendererClass=&NiagaraSpriteRendererProperties}]
 }
 Water_Splash_Arms {
  emitterUpdateScript [
   &EmitterState()
   &SpawnRate(
    SpawnRate=&Multiply_Float(A=&MakeCustomFloatFromBool(Bool=$User.bIsInWater, True Float=1, False Float=0), B=20)
    Use Spawn Probability=false
   )
  ]
  particleSpawnScript [
   &PartitionParticles()
   &InitializeParticle(
    Lifetime Max=0.7
    Lifetime Min=0.5
    Lifetime Mode=::NewEnumerator1
    Mass=&UniformRangedFloat(Minimum=0.6, Maximum=1.2)
    Mass Mode=::NewEnumerator1
    Mesh Scale=&UniformRangedVector(Minimum=<0.15 0.15 0.15>, Maximum=<0.2 0.2 0.2>)
    Mesh Scale Mode=::NewEnumerator4
    UsePositionOffset=false
    Write Material Random=true
   )
   &InitialMeshOrientation(
    Mesh Axis To Orient=&Lerp_Vector(A=<1 1 0>, B=<-1 1 0>, Alpha=&MakeFloatFromInt(Int=$Particles.PartitionParticles.Partition))
    Mesh Orientation Mode=::NewEnumerator2
    Use Rotation Vector=false
   )
   /SetVariables(Particles.MeshOrientation=&MultiplyQuaternion(
     Quaternion A=$Particles.MeshOrientation
     Quaternion B=&MakeQuatFromAxisAngle(Axis=<0 1 0>, AngleInDegrees=&RandomRangeFloat(Minimum=-25, Maximum=25), NormalizeAxis=true)
    )) {isSetParameters=true}
  ]
  particleUpdateScript [
   &ParticleState()
   &SkeletalMeshLocation(
    Bone / Socket Index=$Particles.PartitionParticles.Partition
    MaskSpawnGroup=false
    Mesh Sampling Type=::NewEnumerator5
    Orientation Sampling=::NewEnumerator1
    Sampled Position Offset Coordinate Space:ENiagaraSkelMeshTransforms=::NewEnumerator4
    Skeletal Mesh=$System.SkeletalMesh
    Socket Sampling Mode=::NewEnumerator5
   )
   &DynamicMaterialParameters(
    Index 0 Param 1=&FloatFromCurve(FloatCurve=DataInterface() {propertyValues {Curve {keys [
         {arriveTangent=0.8954, interpMode=RCIM_Cubic, leaveTangent=0.8954, tangentMode=RCTM_User, time=-0.0005, value=-1.845}
         {arriveTangent=1.8419, interpMode=RCIM_Cubic, leaveTangent=1.8419, time=0.7887, value=-0.7711}
         {arriveTangent=3.1029, interpMode=RCIM_Cubic, leaveTangent=3.1029, tangentMode=RCTM_User, time=1.0011, value=0}
        ]}}})
    Index 0 Param 2=&FloatFromCurve(FloatCurve=DataInterface() {propertyValues {Curve {keys [<at -0.0046 4>,<at 1.0011 0>]}}})
    Index 0 Param 3=&FloatFromCurve(
     FloatCurve=DataInterface() {propertyValues {Curve {keys [
         {arriveTangent=0.096, interpMode=RCIM_Cubic, leaveTangent=0.096, tangentMode=RCTM_User, time=-0.0046, value=0.0882}
         <at 1.0011 0.1372 RCIM_Cubic>
        ]}}}
     Scale Curve=&Multiply_Float(A=2, B=$User.EffectScale)
    )
    Index 0 Param 4=&FloatFromCurve(FloatCurve=DataInterface() {propertyValues {Curve {keys [
         {arriveTangent=1.9125, interpMode=RCIM_Cubic, leaveTangent=0.2728, tangentMode=RCTM_User, time=-0.0046, value=0.9276}
         {arriveTangent=4.6326, interpMode=RCIM_Cubic, leaveTangent=4.6326, tangentMode=RCTM_User, time=1.0011, value=2.7648}
        ]}}})
    Module.Index0Param0Write=true
    Module.Index0Param1Write=true
    Module.Index0Param2Write=true
    Module.Index0Param3Write=true
   )
   &ScaleColor(
    Scale Alpha=&FloatFromCurve(FloatCurve=DataInterface() {propertyValues {Curve {keys [<at -0.0046 1.0113 RCIM_Cubic>,<at 1.0011 0.002 RCIM_Cubic>]}}})
    ScaleA=true
    ScaleRGB=false
   )
   &Drag(Drag=2, Use Linear Drag=true, Use Rotational Drag=false)
   &SolveForcesAndVelocity(Clamp Velocity=false, Limit Acceleration=false)
   &ScaleMeshSize(Scale Factor=&VectorFromCurve(VectorCurve=DataInterface() {propertyValues {
       XCurve {keys [<at 0.0044 1.4881>,<at 0.9989 3>]}
       YCurve {keys [<at -0.0012 1.4698>,<at 0.9989 3>]}
       ZCurve {keys [<at 0.0044 1.5612>,<at 0.9989 1.5622>]}
      }})) {module=MeshSizeScale}
   &CameraOffset(Camera Offset Amount=15)
  ]
  rendererClasses [&NiagaraMeshRendererProperties]
  renderers [{rendererClass=&NiagaraMeshRendererProperties, rendererIndex=0}]
 }
}
summary {
 systemName=NS_PeacefulPoemCruise_Skid_Emote
 emitters [
  {emitterName=SkidMarks, rendererClasses [&NiagaraRibbonRendererProperties]}
  {emitterName=Water_Splash_Arms, rendererClasses [&NiagaraMeshRendererProperties]}
  {emitterName=FN_Water_Drops_, rendererClasses [&NiagaraSpriteRendererProperties]}
  {emitterName=SmokeWorld, rendererClasses [&NiagaraSpriteRendererProperties]}
  {emitterName=Chunks, rendererClasses [&NiagaraSpriteRendererProperties]}
  {emitterName=Tire_Skidstreaks, rendererClasses [&NiagaraSpriteRendererProperties]}
 ]
 userVariables [
  User.bIsFrontEnd {type=NiagaraBool}
  User.bIsFrontEndPreview {type=NiagaraBool}
  User.bIsInWater {type=NiagaraBool}
  User.EffectScale {type=NiagaraFloat, defaultValue=0, description=""}
  User.IsWaterDeep {type=NiagaraBool}
  User.WaterBodyIndex {type=NiagaraFloat, defaultValue=0, description=""}
  User.WaterDepth {type=NiagaraFloat, defaultValue=0, description=""}
  User.WaterPlaneLocation {type=NiagaraPosition, description="", defaultValue {}}
  User.WaterPlaneLocation {type=Vector3f, description="", defaultValue {struct=Vector3f, value=<0 0 0>}}
  User.WaterPlaneNormal {type=Vector3f, description="", defaultValue {struct=Vector3f, value=<0 0 0>}}
  User.WaveAttenuationFactor {type=NiagaraFloat, defaultValue=0, description=""}
  User.WaveTime {type=NiagaraFloat, defaultValue=0, description=""}
 ]
}
